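# Translucent unlit, two-sided Niagara sprite material for the smoky energy cloud in the RainbowEnergy skydiving contrail.
# A dynamic parameter picks a frame from a 2x2 inky smoke flipbook, which particle color tints before emissive adjustments.
# Opacity combines particle alpha, the smoke sample, a normalized-age envelope, and a vertical UV mask that fades top and bottom edges.
# A cubed near-camera fade feeds a depth fade to soften camera and geometry intersections.
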Material /BRCosmetics/Athena/Effects/Particles/Skydiving/NS_Contrail_RainbowEnergy/Materials/M_RainbowEnergyCloud
graph {
 properties {
  bUsedWithNiagaraSprites=true
  blendMode=BLEND_Translucent
  pixelDepthOffsetMode=PDOM_Legacy
  shadingModel=MSM_Unlit
  twoSided=true
  usedShadingModels=Unlit
 }
 nodes [
  n1 = Power(Base=MaterialFunctionCall(Distance=ScalarParameter() {defaultValue=100, parameterName=ShrinkDistance}) {materialFunction=&Near_Camera_Fade}.Result) {constExponent=3}
  n2 = DepthFade(Opacity=@n1, FadeDistance=ScalarParameter() {defaultValue=100, parameterName=FadeDist}) {fadeDistanceDefault=12}
  FeatureLevelSwitch(Default=@n2, ES2=@n1, ES3_1=@n1, SM4_REMOVED=@n2)
  n3 = ParticleColor()
  n4 = DynamicParameter() {paramNames [spriteIndex,normAge,randPP,Param4]}
  n5 = Constant() {r=2}
  n6 = TextureSample(UVs=MaterialFunctionCall("Animation  Phase (0-1)"=@n4.spriteIndex, Number of Rows=@n5, Number of Columns=@n5) {materialFunction=&FlipBook}.UVs) {texture=/BRCosmetics/Athena/Effects/Textures/T_InkySmoke_Flipbook}
  n7 = Saturate(None=@n6.RGB)
  MaterialFunctionCall(Emissive=Multiply(A=@n3.RGB, B=@n7)) {materialFunction=&MF_FX_EmissiveAdjustments}
  n8 = Multiply(A=@n6.RGB, B=Multiply(A=Clamp(None=@n4.normAge) {minDefault=0.2}, B=OneMinus(None=@n4.normAge)))
  n9 = MaterialFunctionCall(Float2=TextureCoordinate()) {materialFunction=&BreakOutFloat2Components}
  n10 = OneMinus(None=@n9.G)
  Multiply(A=Multiply(A=Multiply(A=@n3.A, B=@n7), B=Multiply(A=@n8, B=@n2)), B=Multiply(A=Multiply(A=@n9.G, B=@n10)) {constB=5})
 ]
}
