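# Translucent lit Niagara sprite and mesh particle material for the RainbowEnergy skydiving contrail.
# A panned smoke texture, randomized per particle through a dynamic parameter, is tinted by particle color through an LDR-to-HDR remap.
# An opacity chain combines that texture with a V-gradient mask, normalized-age fade, near-camera fade and depth fade.
# An offset chain stretches Z along V and widens along vertex normals using vertex red.
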Material /BRCosmetics/Athena/Effects/Particles/Skydiving/NS_Contrail_RainbowEnergy/Materials/M_RainbowEnergy_Niagara
graph {
 properties {
  bUsedWithNiagaraMeshParticles=true
  bUsedWithNiagaraSprites=true
  blendMode=BLEND_Translucent
  pixelDepthOffsetMode=PDOM_Legacy
  usedShadingModels=Default Lit
 }
 nodes [
  n1 = Constant() {r=1}
  n2 = MaterialFunctionCall(Float2=TextureCoordinate()) {materialFunction=&BreakOutFloat2Components}
  n3 = ScalarParameter() {defaultValue=800, parameterName=height}
  n4 = ScalarParameter() {defaultValue=5, parameterName=WidthMin}
  n5 = ScalarParameter() {defaultValue=50, parameterName=WidthMax}
  n6 = VertexColor()
  n7 = Multiply(A=Multiply(A=VertexNormalWS(), B=LinearInterpolate(A=@n4, B=@n5, Alpha=@n6.R)), B=@n2.G)
  Add(
   A=MaterialFunctionCall(X=@n1, Y=@n1, Z=Multiply(A=@n2.G, B=@n3)) {materialFunction=&MakeFloat3}.Result
   B=@n7
  )
  n8 = VertexColor()
  Add(B=Multiply(A=@n8.R) {constB=3})
  Constant3Vector() {constant=<rgba 1 0 0 0>}
  MaterialFunctionCall() {materialFunction=&BoundingBoxBased_0-1_UVW}
  MaterialFunctionCall(Y=Multiply(A=@n8.B) {constB=-1}) {materialFunction=&MakeFloat2}
  n9 = MaterialFunctionCall(Float2=TextureCoordinate()) {materialFunction=&BreakOutFloat2Components}
  n10 = Multiply(A=@n9.R)
  n11 = Multiply(A=@n9.G) {constB=2}
  n12 = DynamicParameter() {paramNames [ageNorm,randPP,Param3,Param4]}
  n13 = Panner(
   Coordinate=Add(
    A=MaterialFunctionCall(X=@n10, Y=@n11) {materialFunction=&MakeFloat2}.Result
    B=MaterialFunctionCall(X=@n12.randPP, Y=@n12.randPP) {materialFunction=&MakeFloat2}.Result
   ) {constB=0}
   Time=Time()
   Speed=Multiply(A=@n12.randPP) {constB=-1}
  ) {speedY=-1}
  n14 = TextureSample(UVs=@n13) {texture=/BRCosmetics/Athena/Effects/Textures/T_InkySmokeTrail}
  n15 = MaterialFunctionCall(LDR Input=Clamp(None=@n14.RGB), HDR Tint=ParticleColor().RGB) {materialFunction=&DeriveHDRfromLDR}
  MaterialFunctionCall(Emissive=@n15.RemappedRange) {materialFunction=&MF_FX_EmissiveAdjustments}
  Multiply(A=TextureSample(UVs=Multiply(A=@n13) {constB=0.1}) {samplerType=SAMPLERTYPE_Normal, texture=/Game/Global/Textures/Default/Tileable_Patterns/1Normal}.RGB) {constB=10}
  Multiply(A=@n7, B=@n5)
  Multiply(A=@n14.RGB) {constB=20}
  n16 = Multiply(A=Power(Base=OneMinus(None=@n2.G)), B=@n2.G)
  n17 = Multiply(A=Multiply(A=Clamp(None=@n16), B=@n14.RGB), B=OneMinus(None=DynamicParameter() {paramNames [ageNorm,randPP,Param3,Param4]}.ageNorm))
  n18 = MaterialFunctionCall(Distance=ScalarParameter() {defaultValue=100, parameterName=ShrinkDistance}) {materialFunction=&Near_Camera_Fade}
  Multiply(
   A=@n17
   B=DepthFade(Opacity=Power(Base=@n18.Result) {constExponent=3}, FadeDistance=ScalarParameter() {defaultValue=100, parameterName=FadeDist}) {fadeDistanceDefault=12}
  )
  Panner(Coordinate=AppendVector(A=@n2.G, B=@n1), Time=Time(), Speed=VertexColor().G) {speedX=1}
  Reroute(None=@n15.RemappedRange)
  Reroute(None=@n17)
 ]
}
