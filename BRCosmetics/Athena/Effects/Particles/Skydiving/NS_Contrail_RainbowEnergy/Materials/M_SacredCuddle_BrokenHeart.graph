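# Unlit translucent Niagara sprite material for the SacredCuddle broken-heart particle, filed under the RainbowEnergy skydiving contrail effect.
# The heart texture's saturated RGB is tinted by particle color and passed through MF_FX_EmissiveAdjustments for emissive.
# Opacity reuses the texture RGB instead of its alpha, scaled by particle alpha and a parameterized CameraDepthFade, then softened against scene geometry with DepthFade.
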Material /BRCosmetics/Athena/Effects/Particles/Skydiving/NS_Contrail_RainbowEnergy/Materials/M_SacredCuddle_BrokenHeart
graph {
 properties {
  bAutomaticallySetUsageInEditor=false
  bUsedWithNiagaraSprites=true
  blendMode=BLEND_Translucent
  shadingModel=MSM_Unlit
  usedShadingModels=Unlit
 }
 nodes [
  n1 = TextureSample() {texture=/BRCosmetics/Athena/Effects/Particles/Skydiving/NS_Contrail_RainbowEnergy/Textures/T_SacredCuddle_BrokenHeart}
  n2 = ParticleColor()
  n3 = ScalarParameter() {defaultValue=20, parameterName=CameraDepthFade}
  DepthFade(Opacity=Multiply(
    A=Multiply(A=@n1.RGB, B=@n2.A)
    B=MaterialFunctionCall(Fade Length=@n3) {materialFunction=&CameraDepthFade}.Result
   )) {fadeDistanceDefault=5, opacityDefault=15}
  MaterialFunctionCall(Emissive=Multiply(A=Saturate(None=@n1.RGB), B=@n2.RGB)) {materialFunction=&MF_FX_EmissiveAdjustments}
 ]
}
