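# SacredCuddle skydiving contrail: a pink Niagara effect attached to character bones, with hearts implied by variable names.
# Dots, black streaking sprites and pink mesh trails split across four bones, using per-partition arrays to vary strength and size in pairs.
# Contrail setup drives streak velocity, while the Timeline ramps dot rate and tilts front-end motion.
# Two looping particles pulse with waveforms and time-of-day alpha; the attached-particle light emitter is disabled.
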NiagaraSystem /BRCosmetics/Athena/Effects/Particles/Skydiving/NS_Contrail_RainbowEnergy/NS_Dive_SacredCuddle
dependencies {
 usedDataInterfaces [
  &NiagaraDataInterfaceArrayFloat3
  &NiagaraDataInterfaceCurve
  &NiagaraDataInterfaceArrayFloat
  &NiagaraDataInterfaceColorCurve
  &NiagaraDataInterfaceVector2DCurve
  &NiagaraDataInterfaceVectorCurve
 ]
 usedDynamicInputs [
  &SimulationPosition
  &UniformRangedVector2D
  &SetBoolByFloatComparison
 ]
 usedModules [
  &self:SystemSpawnScript.SystemScriptSource.SystemScriptGraph.NiagaraNodeAssignment_0.SetVariables_#1
  &SystemState
  /Game/Effects/Niagara/Modules/Contrails/FN_Contrail_Setup
  &FN_Integrator
  &Timeline
 ]
}
emitters {
 FN_Contrails_AttachedParticle_v2003 {
  bEnabled=false
  emitterUpdateScript [
   &EmitterState()
   /Game/Effects/Niagara/Modules/Spawn/SpawnRateFromMultipleSpawnGroups_2(Global Spawnrate=2)
  ]
  particleSpawnScript [
   &FN_SetSpawnGroup(Number of Spawn Groups=2)
   &InitializeParticle(
    Color=&LerpLinearColors(StartColor=<rgba 0 1.2688 3 1>, EndColor=<rgba 0.7982 0 3 1>, LerpFactor=HlslExpression() {hlslExpression="rand(Particles.UniqueID)"})
    Lifetime=0.5
    Sprite Size Mode=::NewEnumerator3
    Uniform Sprite Size=20
    UsePositionOffset=false
    Write Material Random=true
   ) {module=InitializeParticle001}
  ]
  particleUpdateScript [
   &ParticleState()
   &SampleSkeletalMeshSkeleton(Mesh:NiagaraDataInterfaceSkeletalMesh=$User.SkeletalMesh, Specific Bone Index:NiagaraInt32=$Particles.SpawnGroup)
   /Niagara/Modules/Spawn/Location/SkeletalMeshSkeletonLocation(Module.Inherit Velocity=false, Module.Use Sampled Bone as Ribbon ID=true)
   &SolveForcesAndVelocity(Clamp Velocity=false, Limit Acceleration=false) {enabled=false}
   /SetVariables(Particles.Color=&MakeLinearColorFromVectorAndFloat(
     "Vector (RGB)"=&MakeVectorFromLinearColorRGB(Color=&ColorFromCurve(ColorCurve=DataInterface() {propertyValues {
         AlphaCurve {keys [<at -0.0023 1 RCIM_Cubic>,<at 0.9977 1 RCIM_Cubic>]}
         BlueCurve {keys [<at -0.0764 1 RCIM_Cubic>,<at -0.0015 3>,<at 0.9977 3 RCIM_Cubic>]}
         GreenCurve {keys [<at -0.0764 1 RCIM_Cubic>,<at -0.0015 1.2>,<at 0.9977 0 RCIM_Cubic>]}
         RedCurve {keys [<at -0.0764 1 RCIM_Cubic>,<at -0.0015 0>,<at 0.9977 0.7 RCIM_Cubic>]}
        }}))
     "Float (Alpha)"=&FloatFromCurve(FloatCurve=DataInterface() {propertyValues {Curve {keys [<at 0 0>,<at 0.1 0.9704>,<at 0.9 1>,<at 1.0011 0.002>]}}})
    )) {isSetParameters=true}
   &CameraOffset(Camera Offset Amount=10)
   /SetVariables(Particles.LightRadius=&Lerp_Float(
     A=20
     B=100
     Alpha=&FloatFromCurve(FloatCurve=DataInterface() {propertyValues {Curve {keys [<at -0.0186 0>,<at 0.2433 1>,<at 0.749 1>,<at 1.0011 0.002>]}}})
    )) {isSetParameters=true}
   /SetVariables(Particles.LightExponent:NiagaraFloat=&FloatFromCurve(FloatCurve=DataInterface() {propertyValues {Curve {keys [<at -0.0046 0>,<at 0.3 1>,<at 0.7 1>,<at 1.0011 0.002>]}}})) {isSetParameters=true}
   &ScaleSpriteSize(Scale Factor=&Vector2DFromCurve(
     Vector2Curve=DataInterface() {propertyValues {XCurve {keys [<at 0 0>,<at 0.3 1>,<at 0.7 1>,<at 0.9989 0>]}, YCurve {keys [<at 0 0>,<at 0.3 1>,<at 0.7 1>,<at 0.9989 0>]}}}
     ScaleCurve=<2 2>
    ))
  ]
  rendererClasses [&NiagaraSpriteRendererProperties, &NiagaraLightRendererProperties]
  renderers [
   {rendererClass=&NiagaraSpriteRendererProperties}
   {rendererClass=&NiagaraLightRendererProperties, rendererIndex=1}
  ]
 }
 FN_Contrails_Dots {
  emitterUpdateScript [
   &EmitterState()
   &SpawnRate(SpawnRate=&Lerp_Float(A=8, B=32, Alpha=$System.Timeline.NormalizedTime), Use Spawn Probability=false)
  ]
  particleSpawnScript [
   &FN_SetSpawnGroup() {enabled=false}
   &PartitionParticles(Number Of Partitions=4)
   &InitializeParticle(
    Color=<rgba 1 0.1687 0.4166 1>
    Lifetime Max=1
    Lifetime Min=0.5
    Lifetime Mode=::NewEnumerator1
    Sprite Rotation Mode=::NewEnumerator2
    Sprite Size Mode=::NewEnumerator1
    UsePositionOffset=false
    Write Material Random=true
   ) {module=InitializeParticle001}
   &ShapeLocation(MaskSpawnGroup=false, Override Seed=false, Sphere Radius=5, UseOwnerScale=true) {enabled=false}
   &SkeletalMeshLocation(
    Bone / Socket Index=$Particles.PartitionParticles.Partition
    Bone Sampling Mode=::NewEnumerator4
    MaskSpawnGroup=false
    Skeletal Mesh=$System.Skeletal Mesh
   )
   &AddVelocity(Velocity=&FN_FrontEndVector3Modifier(
     Gameplay=&UniformRangedVector(
      Minimum=&SelectVectorFromArray(
       Direct Array Index=$Particles.PartitionParticles.Partition
       Vector Selection Array=DataInterface() {propertyValues {FloatData [<0 0 100>, <0 0 100>, <0 0 50>, <0 0 50>]}}
      )
      Maximum=&SelectVectorFromArray(
       Direct Array Index=$Particles.PartitionParticles.Partition
       Vector Selection Array=DataInterface() {propertyValues {FloatData [<0 0 200>, <0 0 200>, <0 0 100>, <0 0 100>]}}
      )
     )
     Front End=&Lerp_Vector(
      A=&UniformRangedVector(
       Minimum=&SelectVectorFromArray(
        Direct Array Index=$Particles.PartitionParticles.Partition
        Vector Selection Array=DataInterface() {propertyValues {FloatData [<0 0 100>, <0 0 100>, <0 0 50>, <0 0 50>]}}
       )
       Maximum=&SelectVectorFromArray(
        Direct Array Index=$Particles.PartitionParticles.Partition
        Vector Selection Array=DataInterface() {propertyValues {FloatData [<0 0 200>, <0 0 200>, <0 0 100>, <0 0 100>]}}
       )
      )
      B=&UniformRangedVector(
       Minimum=&SelectVectorFromArray(
        Direct Array Index=$Particles.PartitionParticles.Partition
        Vector Selection Array=DataInterface() {propertyValues {FloatData [<0 -100 100>, <0 -100 100>, <0 -50 50>, <0 -50 50>]}}
       )
       Maximum=&SelectVectorFromArray(
        Direct Array Index=$Particles.PartitionParticles.Partition
        Vector Selection Array=DataInterface() {propertyValues {FloatData [<0 -200 200>, <0 -200 200>, <0 -100 100>, <0 -100 100>]}}
       )
      )
      Alpha=$System.Timeline.NormalizedTime
     )
    ))
  ]
  particleUpdateScript [
   &ParticleState()
   &ScaleSpriteSize(Uniform Curve Sprite Scale=DataInterface() {propertyValues {Curve {keys [
        {interpMode=RCIM_Cubic, tangentMode=RCTM_User, time=0, value=0}
        {interpMode=RCIM_Cubic, tangentMode=RCTM_User, time=0.25, value=1}
        {arriveTangent=-1.3333, interpMode=RCIM_Cubic, leaveTangent=-1.3333, time=0.85, value=1}
        {tangentMode=RCTM_User, time=1, value=0}
       ]}}})
   /Niagara/Modules/Update/Forces/CurlNoiseForce(
    Noise Frequency=0.1
    Noise Strength=&SelectFloatFromArray(
     Direct Array Index=$Particles.PartitionParticles.Partition
     Float Selection Array=DataInterface() {propertyValues {FloatData [500,500,200,200]}}
    )
    Pan Noise=false
   )
   &SolveForcesAndVelocity(Clamp Velocity=false, Limit Acceleration=false)
  ]
  rendererClasses [&NiagaraSpriteRendererProperties]
  renderers [{rendererClass=&NiagaraSpriteRendererProperties}]
 }
 Particles_SkeletalMesh {
  emitterUpdateScript [
   &EmitterState()
   &SpawnRate(SpawnRate=120, Use Spawn Probability=false)
  ]
  particleSpawnScript [
   &PartitionParticles(Number Of Partitions=4)
   &InitializeParticle(
    Color=<rgba 0 0 0 2>
    Lifetime Max=0.6
    Lifetime Min=0.4
    Lifetime Mode=::NewEnumerator1
    Sprite Size Max=<30 100>
    Sprite Size Min=<15 50>
    Sprite Size Mode=::NewEnumerator2
    UsePositionOffset=true
    Write Material Random=true
   )
   &ShapeLocation(
    Apply To Particle Position=::NewEnumerator1
    MaskSpawnGroup=false
    Offset Coordinate Space=::Local
    Offset Mode=::NewEnumerator0
    Override Seed=false
    Sphere Radius=8
    UseOwnerScale=true
   )
   &DynamicMaterialParameters(
    Index 0 Param 1=&UniformRangedFloat(Minimum=0, Maximum=3)
    Module.Param0WriteEnabled=true
    Module.Param1WriteEnabled=false
    Module.Param2WriteEnabled=false
    Module.Param3WriteEnabled=false
   )
  ]
  particleUpdateScript [
   &ParticleState()
   &SkeletalMeshLocation(
    Bone / Socket Index=$Particles.PartitionParticles.Partition
    Bone Sampling Mode=::NewEnumerator4
    MaskSpawnGroup=false
    Orientation Sampling=::NewEnumerator1
    Sampled Position Offset=$Particles.ShapeLocation.ShapePosition
    Skeletal Mesh=$System.Skeletal Mesh
   )
   /SetVariables(Particles.NewAge:NiagaraFloat=&Multiply_Float(A=$Particles.Age, B=0.25)) {isSetParameters=true}
   /SetVariables(Particles.Velocity=&Multiply_VectorByFloat(
     Vector=&Lerp_Vector(A=<0 0 0>, B=$System.FN_Contrail_Setup.Direction Velocity, Alpha=$Particles.NewAge)
     Float=&SelectFloatFromArray(Direct Array Index=$Particles.PartitionParticles.Partition, Float Selection Array=DataInterface() {propertyValues {FloatData [2.5,2.5,1,1]}})
    )) {isSetParameters=true}
   /SetVariables(Particles.Position=&AddVectorToPosition(
     Position=$Particles.Position
     Vector=&Multiply_VectorByFloat(Vector=$Particles.Velocity, Float=$Particles.Age)
    )) {isSetParameters=true}
   &ScaleColor(Scale Alpha=2.5, Scale RGB=<0 0 0>, ScaleA=true, ScaleRGB=true)
   &ScaleSpriteSize(Uniform Curve Scale=&SelectFloatFromArray(
     Direct Array Index=$Particles.PartitionParticles.Partition
     Float Selection Array=DataInterface() {propertyValues {FloatData [1.51,1.51,0.75,0.75]}}
    )) {module=ScaleSpriteSize001}
   &DynamicMaterialParameters(
    Index 0 Param 2=$Particles.NormalizedAge
    Index 0 Param 3=$Particles.MaterialRandom
    Module.Param0WriteEnabled=false
    Module.Param1WriteEnabled=true
    Module.Param2WriteEnabled=true
    Module.Param3WriteEnabled=true
   ) {module=DynamicMaterialParameters001}
   &CameraOffset(Camera Offset Amount=10)
  ]
  rendererClasses [&NiagaraSpriteRendererProperties]
  renderers [{rendererClass=&NiagaraSpriteRendererProperties}]
 }
 Pink_Spawning {
  emitterUpdateScript [
   &EmitterState()
   &SpawnRate(SpawnRate=45, Use Spawn Probability=false)
  ]
  particleSpawnScript [
   &PartitionParticles(Number Of Partitions=4)
   &InitializeParticle(
    Color=&UniformRangedLinearColor(Minimum=<rgba 5 1.4231 4.7108 1>, Maximum=<rgba 500 0 254.1108 1>)
    Mesh Scale Max=<1 1 2>
    Mesh Scale Min=<0.2 0.2 0.2>
    Mesh Scale Mode=::NewEnumerator2
    Sprite Size Mode=::NewEnumerator3
    Uniform Sprite Size=30
    UsePositionOffset=false
    Write Material Random=true
   )
  ]
  particleUpdateScript [
   &ParticleState()
   &SkeletalMeshLocation(
    Bone / Socket Index=$Particles.PartitionParticles.Partition
    Bone Sampling Mode=::NewEnumerator4
    MaskSpawnGroup=false
    Skeletal Mesh=$System.Skeletal Mesh
   )
   /BRCosmetics/Athena/Effects/Particles/Skydiving/NS_Contrail_AmmoBelt/Module/FN_OrientMeshToEngineOwner(
    Mesh Up Vector:Vector3f=&FN_FrontEndVector3Modifier(
     Gameplay=<0 0 -1>
     Front End=&TransformVector(
      Vector=&Lerp_Vector(A=<0 0 -1>, B=<0 0.75 -1>, Alpha=$System.Timeline.NormalizedTime)
      Source Space=::World
      Destination Space=::Local
     )
    )
    Set Mesh Orientation=true
   )
   &ScaleMeshSize(Scale Factor=&VectorFromCurve(
     VectorCurve=DataInterface() {propertyValues {XCurve {keys [<at 0 0.25>,<at 1 1>]}, YCurve {keys [<at 0 0.25>,<at 1 1>]}, ZCurve {keys [<at 0 0.25>,<at 1 1>]}}}
     Scale Curve=&SelectVectorFromArray(
      Direct Array Index=$Particles.PartitionParticles.Partition
      Vector Selection Array=DataInterface() {propertyValues {FloatData [<50 50 300>, <50 50 300>, <15 15 100>, <15 15 100>]}}
     )
    )) {module=ScaleMeshSize001}
   &DynamicMaterialParameters(
    Index 0 Param 1=$Particles.NormalizedAge
    Index 0 Param 2=HlslExpression() {hlslExpression=Particles.MaterialRandom*0.5 + 0.5}
    Module.Param0WriteEnabled=true
    Module.Param1WriteEnabled=true
    Module.Param2WriteEnabled=true
    Module.Param3WriteEnabled=true
   )
  ]
  rendererClasses [&NiagaraMeshRendererProperties, &NiagaraSpriteRendererProperties]
  renderers [
   {rendererClass=&NiagaraMeshRendererProperties, rendererIndex=0}
   {rendererClass=&NiagaraSpriteRendererProperties, rendererIndex=1}
  ]
 }
 Pink_Spawning001 {
  emitterUpdateScript [
   &EmitterState(Life Cycle Mode=::NewEnumerator1, Loop Behavior=::NewEnumerator1, Loop Duration Mode=::NewEnumerator1, UseLoopDelay=false)
   &SpawnBurst_Instantaneous(Spawn Count=2, Use Spawn Probability=false, UseLoopCountLimit=false)
  ]
  particleSpawnScript [
   &PartitionParticles()
   &InitializeParticle(
    Color=<rgba 1 0.0838 0.2755 1>
    Lifetime=0.3
    Sprite Size Mode=::NewEnumerator3
    Uniform Sprite Size=65
    UsePositionOffset=false
    Write Material Random=true
   )
  ]
  particleUpdateScript [
   &ParticleState(Kill Particles When Lifetime Has Elapsed=false, Loop Particles Lifetime=true)
   &SkeletalMeshLocation(
    Bone / Socket Index=$Particles.PartitionParticles.Partition
    Bone Sampling Mode=::NewEnumerator4
    MaskSpawnGroup=false
    Sampled Position Offset=&FN_FrontEndVector3Modifier(Gameplay=<0 0 10>, Front End=<0 0 5>)
    Sampled Position Offset Coordinate Space:ENiagaraSkelMeshTransforms=::NewEnumerator2
    Skeletal Mesh=$System.Skeletal Mesh
   )
   /SetVariables(
    Particles.HeartColor:LinearColor=<rgba 1 0.1687 0.4166 1>
    Particles.HeartPos:NiagaraPosition=&AddVectorToPosition(
     Position=$Particles.SkeletalMeshLocation.SampledPosition
     Vector=&FN_FrontEndVector3Modifier(Gameplay=<0 0 15>, Front End=<0 0 10>)
    )
    Particles.HeartSize:Vector2f=&Multiply_Vector2D_ByFloat(
     Float=&Waveform("Waveform [1]"=::NewEnumerator0, "[1] Frequency"=2, "[1] Amplitude Min/Max"=<0.75 1>)
     Vector2D=<20 20>
    )
   ) {isSetParameters=true}
   &CameraOffset(Camera Offset Amount=10)
   &ScaleColor(
    Scale Alpha=&Waveform(
     "Waveform [1]"=::NewEnumerator0
     "[1] Frequency"=2
     "[1] Amplitude Min/Max"=&Multiply_Vector2D_ByFloat(
      Float=&FNTimeOfDayScalar(Night Scale=0.5, Morning Scale=0.85, Day Scale=0.85, Evening Scale=0.5)
      Vector2D=<0.5 1.5>
     )
    )
    Scale RGB=&FN_FrontEndVector3Modifier(Gameplay=<1 1 1>, Front End=<0.25 0.25 0.25>)
    ScaleA=true
    ScaleRGB=true
   )
  ]
  rendererClasses [&NiagaraMeshRendererProperties, &NiagaraSpriteRendererProperties]
  renderers [
   {rendererClass=&NiagaraMeshRendererProperties, rendererIndex=0}
   {rendererClass=&NiagaraSpriteRendererProperties, rendererIndex=1}
   {rendererClass=&NiagaraSpriteRendererProperties, rendererIndex=2}
  ]
 }
}
summary {
 systemName=NS_Dive_SacredCuddle
 emitters [
  {emitterName=FN_Contrails_Dots, rendererClasses [&NiagaraSpriteRendererProperties]}
  {
   bEnabled=false
   emitterName=FN_Contrails_AttachedParticle_v2003
   rendererClasses [&NiagaraSpriteRendererProperties, &NiagaraLightRendererProperties]
  }
  {emitterName=Pink_Spawning, rendererClasses [&NiagaraMeshRendererProperties, &NiagaraSpriteRendererProperties]}
  {emitterName=Particles_SkeletalMesh, rendererClasses [&NiagaraSpriteRendererProperties]}
  {
   emitterName=Pink_Spawning001
   rendererClasses [&NiagaraMeshRendererProperties, &NiagaraSpriteRendererProperties]
  }
 ]
 userVariables [
  User.bIsFrontEnd {type=NiagaraBool}
  User.bIsFrontEndPreview {type=NiagaraBool}
  User.ParaGlideLeanAlpha {type=NiagaraFloat, defaultValue=0, description=""}
  User.RibbonIDOffset {type=NiagaraInt32}
  User.SkeletalMesh {
   type=NiagaraDataInterfaceSkeletalMesh
   defaultValue {
    struct=NiagaraExt_VariableValue_DataInterface
    value {dataInterface=&self:NiagaraDataInterfaceSkeletalMesh_0}
   }
  }
 ]
}
